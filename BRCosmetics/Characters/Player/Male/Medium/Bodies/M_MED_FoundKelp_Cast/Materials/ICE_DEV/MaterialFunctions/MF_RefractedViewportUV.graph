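# Material function that converts a world-space refraction vector into a viewport UV for the refracted scene point, kept in a character cosmetic's ICE_DEV materials.
# It offsets world position along that vector by MF_DistanceToRefractedScene's distance, then projects to screen via custom WorldToClip HLSL.
# A second pass re-queries distance at that UV and reprojects, iteratively refining the screen-space hit.
# The vector input previews as the negated camera vector.
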MaterialFunction /BRCosmetics/Characters/Player/Male/Medium/Bodies/M_MED_FoundKelp_Cast/Materials/ICE_DEV/MaterialFunctions/MF_RefractedViewportUV
graph {
 properties {previewBlendMode=BLEND_Translucent}
 nodes [
  n1 = WorldPosition()
  MaterialExpressionAdd_1 = Add(
   A=Multiply(
    A=MaterialFunctionCall!(Viewport UV) {materialFunction=/Game/Global/MaterialFunctions/TechArt/Refraction/MF_DistanceToRefractedScene}."Distance (S)"
    B=NamedRerouteUsage()
   )
   B=@n1.XYZ
  ) {desc=ScenePos WS}
  n2 = Custom!(PointWS=@MaterialExpressionAdd_1) {
   code="// transform to clip space\r\nfloat4 PointCS = mul(float4(PointWS.xyz,1), \r\n                 LWCToFloat(ResolvedView.WorldToClip));\r\n\r\n// clip to NDC\r\nfloat2 PointNDC = (PointCS.xy / PointCS.w);\r\n\r\n// remapped to [0.0, 1.0]\r\nfloat2 ViewportUV = PointNDC * 0.5 + 0.5;\r\nreturn float2(ViewportUV.x, 1.0-ViewportUV.y);"
   description=Refracted Point to ViewportUV
   outputType=CMOT_Float2
   inputs [{
     inputName=PointWS
     input {expression=&self:MaterialExpressionAdd_1}
    }]
  }
  MaterialExpressionAdd_2 = Add(
   A=Multiply(
    A=MaterialFunctionCall(Viewport UV=@n2) {materialFunction=&MF_DistanceToRefractedScene}."Distance (S)"
    B=NamedRerouteUsage()
   )
   B=@n1.XYZ
  ) {desc=Refracted ScenePos WS}
  n3 = Custom(PointWS=@MaterialExpressionAdd_2) {
   code="// transform to clip space\r\nfloat4 PointCS = mul(float4(PointWS.xyz,1), \r\n                 LWCToFloat(ResolvedView.WorldToClip));\r\n\r\n// clip to NDC\r\nfloat2 PointNDC = (PointCS.xy / PointCS.w);\r\n\r\n// remapped to [0.0, 1.0]\r\nfloat2 ViewportUV = PointNDC * 0.5 + 0.5;\r\nreturn float2(ViewportUV.x, 1.0-ViewportUV.y);"
   description=Refracted Point to ViewportUV
   outputType=CMOT_Float2
   inputs [{
     inputName=PointWS
     input {expression=&self:MaterialExpressionAdd_2}
    }]
  }
  FunctionOutput(None=@n3) {outputName=ViewportUV}
  n4 = CameraVectorWS()
  NamedRerouteDeclaration(None=FunctionInput(Preview=Multiply(A=@n4) {constB=-1}) {inputName=Refraction Vector WS}) {name="Refraction Vector WS (V3)", nodeColor=<rgba 0 1 0.9766 1>}
 ]
}
