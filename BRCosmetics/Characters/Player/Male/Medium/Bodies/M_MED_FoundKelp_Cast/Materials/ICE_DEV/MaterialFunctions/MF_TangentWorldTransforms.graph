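# Material function that converts a vector between tangent and world space, returning both directions from one input, stored with the FoundKelp character's dev materials.
# Tangent, bitangent and normal inputs default to the TangentBasis function's axes and can be overridden.
# Tangent-to-world sums basis axes scaled by the vector's components, while world-to-tangent dots the vector with each axis.
# Both outputs are normalized, so input magnitude is lost.
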MaterialFunction /BRCosmetics/Characters/Player/Male/Medium/Bodies/M_MED_FoundKelp_Cast/Materials/ICE_DEV/MaterialFunctions/MF_TangentWorldTransforms
graph {expressions {
  MaterialExpressionAdd_0 {
   class=MaterialExpressionAdd
   inputs [
    A {from=MaterialExpressionMultiply_0}
    B {from=MaterialExpressionAdd_1}
   ]
  }
  MaterialExpressionAdd_1 {
   class=MaterialExpressionAdd
   inputs [
    A {from=MaterialExpressionMultiply_1}
    B {from=MaterialExpressionMultiply_2}
   ]
  }
  MaterialExpressionConvert_0 {
   class=MaterialExpressionConvert
   inputs [None {from=MaterialExpressionNamedRerouteUsage_4}]
   properties {
    nodeName=Break Float3
    convertInputs [{
      type=Vector3
      expressionInput {
       expression=&self:MaterialExpressionNamedRerouteUsage_4
       inputName=None
       mask=0
       maskA=0
       maskB=0
       maskG=0
       maskR=0
       outputIndex=0
      }
     }]
    convertOutputs [{type=Scalar}, {type=Scalar}, {type=Scalar}]
   }
  }
  MaterialExpressionConvert_3 {
   class=MaterialExpressionConvert
   inputs [
    None {from=MaterialExpressionDotProduct_1}
    None {from=MaterialExpressionDotProduct_2}
    None {from=MaterialExpressionDotProduct_4}
   ]
   properties {
    nodeName=Make Vector3
    convertInputs [
     {
      type=Scalar
      expressionInput {
       expression=&self:MaterialExpressionDotProduct_1
       inputName=None
       mask=0
       maskA=0
       maskB=0
       maskG=0
       maskR=0
       outputIndex=0
      }
     }
     {
      type=Scalar
      expressionInput {
       expression=&self:MaterialExpressionDotProduct_2
       inputName=None
       mask=0
       maskA=0
       maskB=0
       maskG=0
       maskR=0
       outputIndex=0
      }
     }
     {
      type=Scalar
      expressionInput {
       expression=&self:MaterialExpressionDotProduct_4
       inputName=None
       mask=0
       maskA=0
       maskB=0
       maskG=0
       maskR=0
       outputIndex=0
      }
     }
    ]
    convertOutputs [{type=Vector3}]
   }
  }
  MaterialExpressionDotProduct_1 {
   class=MaterialExpressionDotProduct
   inputs [
    A {from=MaterialExpressionNamedRerouteUsage_0, output=""}
    B {from=MaterialExpressionNamedRerouteUsage_1, output=""}
   ]
   properties {}
  }
  MaterialExpressionDotProduct_2 {
   class=MaterialExpressionDotProduct
   inputs [
    A {from=MaterialExpressionNamedRerouteUsage_0, output=""}
    B {from=MaterialExpressionNamedRerouteUsage_2, output=""}
   ]
   properties {}
  }
  MaterialExpressionDotProduct_4 {
   class=MaterialExpressionDotProduct
   inputs [
    A {from=MaterialExpressionNamedRerouteUsage_0, output=""}
    B {from=MaterialExpressionNamedRerouteUsage_3, output=""}
   ]
   properties {}
  }
  MaterialExpressionFunctionInput_0 {
   class=MaterialExpressionFunctionInput
   inputs [Preview {from="", output=""}]
   properties {inputName="Vector In (WS or TS)", previewValue=<0 0 1 1>}
  }
  MaterialExpressionFunctionInput_1 {
   class=MaterialExpressionFunctionInput
   inputs [Preview {from=MaterialExpressionMaterialFunctionCall_0, output=X U}]
   properties {
    bUsePreviewValueAsDefault=true
    description="Default: normalize( TangentToWorld(1.0, 0.0, 0.0) )"
    inputName=Tangent WS
    sortPriority=10
    previewValue=<1 0 0 1>
   }
  }
  MaterialExpressionFunctionInput_2 {
   class=MaterialExpressionFunctionInput
   inputs [Preview {from=MaterialExpressionMaterialFunctionCall_0, output=Y V}]
   properties {
    bUsePreviewValueAsDefault=true
    description="Default: normalize( TangentToWorld(0.0, 1.0, 0.0) )"
    inputName=Bitangent WS
    sortPriority=20
    previewValue=<0 1 0 1>
   }
  }
  MaterialExpressionFunctionInput_3 {
   class=MaterialExpressionFunctionInput
   inputs [Preview {from=MaterialExpressionMaterialFunctionCall_0, output=Z W}]
   properties {
    bUsePreviewValueAsDefault=true
    description="Default: normalize( TangentToWorld(0.0, 0.0, 1.0) )"
    inputName=Normal WS
    sortPriority=30
    previewValue=<0 0 1 1>
   }
  }
  MaterialExpressionFunctionOutput_0 {
   class=MaterialExpressionFunctionOutput
   inputs [None {from=MaterialExpressionNormalize_0, output=""}]
   properties {outputName="Vector Out TS->WS (V3)"}
  }
  MaterialExpressionFunctionOutput_1 {
   class=MaterialExpressionFunctionOutput
   inputs [None {from=MaterialExpressionNormalize_1, output=""}]
   properties {outputName="Vector Out WS->TS (V3)", sortPriority=10}
  }
  MaterialExpressionMaterialFunctionCall_0 {
   class=MaterialExpressionMaterialFunctionCall
   inputs []
   properties {materialFunction=&TangentBasis}
  }
  MaterialExpressionMultiply_0 {
   class=MaterialExpressionMultiply
   inputs [
    A {from=MaterialExpressionConvert_0}
    B {from=MaterialExpressionNamedRerouteUsage_5}
   ]
  }
  MaterialExpressionMultiply_1 {
   class=MaterialExpressionMultiply
   inputs [
    A {from=MaterialExpressionConvert_0}
    B {from=MaterialExpressionNamedRerouteUsage_6}
   ]
  }
  MaterialExpressionMultiply_2 {
   class=MaterialExpressionMultiply
   inputs [
    A {from=MaterialExpressionConvert_0}
    B {from=MaterialExpressionNamedRerouteUsage_7}
   ]
  }
  MaterialExpressionNamedRerouteDeclaration_0 {
   class=MaterialExpressionNamedRerouteDeclaration
   inputs [None {from=MaterialExpressionFunctionInput_1}]
   properties {name="Tangent WS (V3)", nodeColor=<rgba 0 0.2969 1 1>}
  }
  MaterialExpressionNamedRerouteDeclaration_1 {
   class=MaterialExpressionNamedRerouteDeclaration
   inputs [None {from=MaterialExpressionFunctionInput_2}]
   properties {name="Bitangent WS (V3)", nodeColor=<rgba 0.6172 1 0 1>}
  }
  MaterialExpressionNamedRerouteDeclaration_2 {
   class=MaterialExpressionNamedRerouteDeclaration
   inputs [None {from=MaterialExpressionFunctionInput_3}]
   properties {name="Normal WS (V3)", nodeColor=<rgba 1 0 0.9375 1>}
  }
  MaterialExpressionNamedRerouteDeclaration_3 {
   class=MaterialExpressionNamedRerouteDeclaration
   inputs [None {from=MaterialExpressionFunctionInput_0}]
   properties {name="Vector In (V3)", nodeColor=<rgba 0 1 0.7422 1>}
  }
  MaterialExpressionNamedRerouteUsage_0 {class=MaterialExpressionNamedRerouteUsage}
  MaterialExpressionNamedRerouteUsage_1 {class=MaterialExpressionNamedRerouteUsage}
  MaterialExpressionNamedRerouteUsage_2 {class=MaterialExpressionNamedRerouteUsage}
  MaterialExpressionNamedRerouteUsage_3 {class=MaterialExpressionNamedRerouteUsage}
  MaterialExpressionNamedRerouteUsage_4 {class=MaterialExpressionNamedRerouteUsage}
  MaterialExpressionNamedRerouteUsage_5 {class=MaterialExpressionNamedRerouteUsage}
  MaterialExpressionNamedRerouteUsage_6 {class=MaterialExpressionNamedRerouteUsage}
  MaterialExpressionNamedRerouteUsage_7 {class=MaterialExpressionNamedRerouteUsage}
  MaterialExpressionNormalize_0 {class=MaterialExpressionNormalize, inputs [VectorInput {from=MaterialExpressionAdd_0, output=""}]}
  MaterialExpressionNormalize_1 {class=MaterialExpressionNormalize, inputs [VectorInput {from=MaterialExpressionConvert_3, output=""}]}
 }}
