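# Material function that outputs a normalized scalar thickness from a normal, an inward-pointing direction and an index of refraction, for a character cosmetic material.
# It derives named terms for the absolute cosine angle, facing-ratio exponent, parallel and perpendicular intensities, and a Fresnel-style power falloff.
# The output scales an interpolated value by a biased absolute term, an analytic approximation using tuned curves.
# IOR is clamped just above one to avoid an abrupt change.
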MaterialFunction /BRCosmetics/Characters/Player/Male/Medium/Bodies/M_MED_FoundKelp_Cast/Materials/ICE_DEV/MaterialFunctions/MF_ThicknessDirectional
graph {nodes [
  n1 = NamedRerouteUsage()
  n2 = NamedRerouteUsage()
  n3 = NamedRerouteUsage()
  n4 = Multiply(A=Abs(None=NamedRerouteUsage())) {constB=0.5}
  FunctionOutput(None=Multiply(A=LinearInterpolate(A=@n1, B=@n2, Alpha=@n3), B=Add(A=@n4) {constB=0.5})) {outputName="Thickness Normalized (S)"}
  n5 = NamedRerouteUsage()
  n6 = NamedRerouteUsage()
  NamedRerouteDeclaration(None=Abs(None=DotProduct(A=@n5, B=@n6))) {name="CosTheta (S)", nodeColor=<rgba 1 0.0557 0 1>}
  n7 = NamedRerouteUsage()
  n8 = Saturate(None=Multiply(A=Subtract(A=@n7) {constB=0.5}) {constB=0.25})
  NamedRerouteDeclaration(None=LinearInterpolate(Alpha=Power(Base=@n8) {constExponent=0.8}) {constA=1, constB=50}) {name="Facing Ratio Exp (S)", nodeColor=<rgba 0 0.9062 1 1>}
  NamedRerouteDeclaration(None=LinearInterpolate(Alpha=@n8) {constA=1, constB=0.85}) {name="Parallel Intensity (S)", nodeColor=<rgba 1 0.7734 0 1>}
  NamedRerouteDeclaration(None=LinearInterpolate(Alpha=Power(Base=@n8) {constExponent=0.6}) {constA=0.5, constB=0.2}) {name="Perpendicular Intensity (S)", nodeColor=<rgba 0.4531 0 1 1>}
  n9 = FunctionInput(Preview=VertexNormalWS()) {bUsePreviewValueAsDefault=true, description=Default : VertexNormalWS, inputName=Normal, previewValue=<0 0 1 1>}
  NamedRerouteDeclaration(None=@n9) {name="Normal (V3)", nodeColor=<rgba 0 1 0.1328 1>}
  n10 = FunctionInput(Preview=Multiply(A=CameraVectorWS()) {constB=-1}) {
   bUsePreviewValueAsDefault=true
   description="Pointing inside the surface; Default : CameraVectorWS"
   inputName=Direction
   sortPriority=10
   previewValue=<0 0 1 1>
  }
  NamedRerouteDeclaration(None=Multiply(A=@n10) {constB=-1}) {name="Direction (V3)", nodeColor=<rgba 1 0 0.1875 1>}
  n11 = FunctionInput() {
   bUsePreviewValueAsDefault=true
   description=Default : CameraVectorWS
   inputName=IOR
   inputType=FunctionInput_Scalar
   sortPriority=20
   previewValue=<1 0 0 1>
  }
  NamedRerouteDeclaration(None=Max(A=@n11) {constB=1.01, desc="avoids an abrupt change\r\nbetween 1.0 and 1.01"}) {name="IOR (S)", nodeColor=<rgba 0 0.5078 1 1>}
  n12 = Multiply(A=SmoothStep(Value=NamedRerouteUsage()) {constMin=0.5}) {constB=-5}
  n13 = NamedRerouteUsage()
  NamedRerouteDeclaration(None=Power(Base=NamedRerouteUsage(), Exp=Add(A=@n12, B=@n13))) {name="Fresnel Exponent (S)", nodeColor=<rgba 0 0.7656 1 1>}
 ]}
