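# Material function that adds refractive volumetric ice to an input material for the FoundKelp character cosmetic, with an ice mask input.
# Detail normals, IOR and view-dependent thickness feed refraction and a dielectric interior volume call computing absorption and scattering under sun and sky light.
# Static switches select opaque or alpha-blended refraction, using scene color or estimated scene luminance.
# Shading-path switches and KFC_UseMobile provide mobile fallbacks, including a constant MobileDistortion color.
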MaterialFunction /BRCosmetics/Characters/Player/Male/Medium/Bodies/M_MED_FoundKelp_Cast/Materials/ICE_DEV_UPDATE/MF_VolumetricIce
graph {nodes [
  FunctionOutput(None=BlendMaterialAttributes(A=NamedRerouteUsage(), B=NamedRerouteUsage(), Alpha=NamedRerouteUsage())) {outputName=Material}
  n1 = VectorParameter() {
   group=Ice | 04_Volume
   parameterName=Volume - Absorption Coeffs
   sortPriority=20
   channelNames {a=Global Amount, b=Absorption B, g=Absorption G, r=Absorption R}
  }
  NamedRerouteDeclaration(None=Multiply(A=@n1.RGB, B=@n1.RGB)) {name="Absorption Coeffs - Constant (V3)", nodeColor=<rgba 0.125 0 1 1>}
  n2 = VectorParameter() {
   group=Ice | 04_Volume
   parameterName=Volume - Scattering Coeffs
   sortPriority=30
   channelNames {a=Global Amount, b=Scattering B, g=Scattering G, r=Scattering R}
  }
  n3 = Multiply(
   A=ConstantBiasScale(None=NamedRerouteUsage()) {bias=-0.5, scale=2}
   B=ScalarParameter() {group=Ice | 04_Volume, parameterName=Volume - AO Scattering Variation, sortPriority=35}
  )
  NamedRerouteDeclaration(None=Multiply(A=Multiply(A=@n2.RGB, B=@n2.RGB), B=Add(A=@n3))) {name="Scattering Coeffs - Constant (V3)", nodeColor=<rgba 0.1953 1 0 1>}
  n4 = VectorParameter() {
   group=Ice | 01_Surface
   parameterName=Surface - Detail Normal Controls
   sortPriority=30
   channelNames {a="Min Intensity (Facing)", b="Max Intensity (Grazing)", g=Fresnel Exp, r=UV Scale}
   defaultValue=<rgba 0.2 5 2 0.6>
  }
  NamedRerouteDeclaration(None=@n4.UV Scale) {name="Detail Normal UV Scale (S)", nodeColor=<rgba 0 1 0.1328 1>}
  n5 = VectorParameter() {
   group=Ice | 01_Surface
   parameterName=Surface - AO
   sortPriority=60
   channelNames {a=n/a, b=Exp, g=Multiply, r=Subtract}
   defaultValue=<rgba 0 1 1 0>
  }
  NamedRerouteDeclaration(None=@n5.Subtract) {name="Surface AO Subtract (S)", nodeColor=<rgba 1 0 0.3281 1>}
  n6 = NamedRerouteUsage()
  n7 = NamedRerouteUsage()
  NamedRerouteDeclaration(None=Add(A=Multiply(A=@n6, B=@n7), B=NamedRerouteUsage())) {name="Lighting - Sky Illuminance (Volume) (V3)", nodeColor=<rgba 0 1 0.4141 1>}
  n8 = MaterialFunctionCall!(
   Specular Luminance
   Diffuse Luminance
   Surface Normal WS=NamedRerouteUsage()
   Surface Specular=NamedRerouteUsage()
   Surface Roughness=NamedRerouteUsage()
   Surface Opacity=NamedRerouteUsage()
   Refraction Vector WS=NamedRerouteUsage()
   Scene Luminance=NamedRerouteUsage()
   Interior Emission=NamedRerouteUsage()
   Volume Thickness=NamedRerouteUsage()
   Volume Density=NamedRerouteUsage()
   Absorption Coeffs=NamedRerouteUsage()
   Scattering Coeffs=NamedRerouteUsage()
   Scattering Phase=NamedRerouteUsage()
   Is Scene Inside Volume=NamedRerouteUsage()
   Interface Thickness=NamedRerouteUsage()
   //---- ALPHA-BLENDING ONLY ----//
   Haze Amount=NamedRerouteUsage()
   //----------- LIGHTING -----------//
   Sky Illuminance=NamedRerouteUsage()
   Sun Illuminance=NamedRerouteUsage()
   Sun Direction=NamedRerouteUsage()
  ) {materialFunction=/BRCosmetics/Characters/Player/Male/Medium/Bodies/M_MED_FoundKelp_Cast/Materials/ICE_DEV/MaterialFunctions/MF_DielectricInteriorVolume}
  NamedRerouteDeclaration(None=AppendVector(A=@n8."Refraction Result - Opaque (V3)", B=Constant() {r=1})) {name="Refraction Output (Opaque) (V4)", nodeColor=<rgba 1 0 0.4922 1>}
  n9 = NamedRerouteUsage()
  NamedRerouteDeclaration(None=AppendVector(A=@n8."Refraction Result - AlphaBlending (V3)", B=LinearInterpolate(A=@n8."Material Opacity - AlphaBlending (S)", Alpha=@n9))) {name="Refraction Output (AlphaBlending) (V4)", nodeColor=<rgba 1 0 0.4922 1>}
  n10 = NamedRerouteUsage()
  n11 = MaterialFunctionCall!(Refraction Vector WS=NamedRerouteUsage()) {materialFunction=/BRCosmetics/Characters/Player/Male/Medium/Bodies/M_MED_FoundKelp_Cast/Materials/ICE_DEV/MaterialFunctions/MF_RefractedViewportUV}
  n12 = ScreenPosition()
  n13 = Subtract(A=@n11.ViewportUV, B=@n12.ViewportUV)
  NamedRerouteDeclaration(None=Add(A=Multiply(A=@n10, B=@n13), B=@n12.ViewportUV)) {name="Viewport UV (Refraction) (V2)", nodeColor=<rgba 1 0 0.8438 1>}
  NamedRerouteDeclaration(None=@n5.Multiply) {name="Surface AO Multiply (S)", nodeColor=<rgba 0 0.6484 1 1>}
  NamedRerouteDeclaration(None=@n5.Exp) {name="Surface AO Exp (S)", nodeColor=<rgba 0.9688 1 0 1>}
  n14 = TextureObjectParameter() {
   group=Ice | 01_Surface
   parameterName=Surface - Detail Normal Map
   samplerType=SAMPLERTYPE_Normal
   sortPriority=20
   texture=&FlatNormalBlue
  }
  NamedRerouteDeclaration(None=@n14) {name="Detail Normal Map (TO)", nodeColor=<rgba 0.6406 1 0 1>}
  n15 = VectorParameter() {
   group=Ice | 05_Lighting
   parameterName=Refracted Skybox Luminance
   sortPriority=10
   channelNames {a=n/a, b=Scattering Contribution, g=Enable Extra Illum Inside Volume, r=Extra Illum Behind Volume}
  }
  NamedRerouteDeclaration(None=@n15.Enable Extra Illum Inside Volume) {name="Refracted Skybox Enable Interior Contribution (S)", nodeColor=<rgba 0.9453 1 0 1>}
  n16 = NamedRerouteUsage()
  NamedRerouteDeclaration(None=MaterialFunctionCall!(View Vector=CameraVectorWS(), Normal=NamedRerouteUsage(), "Relative IOR (Eta)"=Divide(B=@n16)) {materialFunction=/Game/Global/MaterialFunctions/TechArt/Refraction/MF_Refract}."Refraction Vector (V3)") {name="Refraction Vector WS (V3)", nodeColor=<rgba 0 1 0.9766 1>}
  NamedRerouteDeclaration(None=@n15.Extra Illum Behind Volume) {name="Refracted Skybox Add Amount (S)", nodeColor=<rgba 0.2891 1 0 1>}
  NamedRerouteDeclaration(None=@n15.Scattering Contribution) {name="Refracted Skybox Scattering Contribution (S)", nodeColor=<rgba 1 0 0.6094 1>}
  n17 = VectorParameter() {
   group=Ice | 01_Surface
   parameterName=Surface - Specular Controls
   sortPriority=40
   channelNames {a=Roughness Mult, b=Roughness Exp, g=Specular Mult, r=Specular Exp}
   defaultValue=<rgba 1 1 1 1>
  }
  NamedRerouteDeclaration(None=@n17.Specular Mult) {name="Specular Mult (S)", nodeColor=<rgba 1 0 0.3047 1>}
  NamedRerouteDeclaration(None=Constant()) {name="Interior Emission (V3)", nodeColor=<rgba 1 0.9197 0 1>}
  n18 = Multiply(A=NamedRerouteUsage(), B=LinearInterpolate(B=NamedRerouteUsage(), Alpha=NamedRerouteUsage()) {constA=1})
  n19 = NamedRerouteUsage()
  NamedRerouteDeclaration(None=Add(A=Multiply(A=@n18, B=@n19), B=NamedRerouteUsage())) {name="Scene Luminance (Volume) (V3)", nodeColor=<rgba 0.5565 0 1 1>}
  n20 = ScalarParameter() {defaultValue=1, group=Ice | 03_Thickness, parameterName=Thickness - Thickness Gradient Exp, sliderMax=10, sliderMin=0.01, sortPriority=10}
  NamedRerouteDeclaration(None=@n20) {name="Thickness Exp (S)", nodeColor=<rgba 1 0.375 0 1>}
  NamedRerouteDeclaration(None=Saturate(None=FunctionInput() {bUsePreviewValueAsDefault=true, inputName=Surface Opacity, inputType=FunctionInput_Scalar, sortPriority=20})) {name="Surface Opacity In (S)", nodeColor=<rgba 0.2656 1 0 1>}
  NamedRerouteDeclaration(None=ScalarParameter() {defaultValue=1, group=Ice | 04_Volume, parameterName=Volume - Density, sortPriority=10}) {name="Volume Density (S)", nodeColor=<rgba 0 0.3438 1 1>}
  NamedRerouteDeclaration(None=ScalarParameter() {group=Ice | 04_Volume, parameterName=Volume - Scattering Phase, sliderMax=1, sliderMin=-1, sortPriority=40}) {name="Scattering Phase (S)", nodeColor=<rgba 0 1 0.6719 1>}
  n21 = Power(Base=NamedRerouteUsage(), Exp=NamedRerouteUsage())
  n22 = NamedRerouteUsage()
  NamedRerouteDeclaration(None=Saturate(None=Multiply(A=@n21, B=@n22))) {name="Surface Specular (S)", nodeColor=<rgba 0.1953 0 1 1>}
  n23 = FunctionInput() {bUsePreviewValueAsDefault=true, inputName=Ice Mask, inputType=FunctionInput_Scalar, sortPriority=10, previewValue=<1 0 0 1>}
  n24 = Saturate(None=@n23)
  NamedRerouteDeclaration(None=LinearInterpolate(
    A=NamedRerouteDeclaration(None=@n24) {name="Ice Mask (Raw) (S)", nodeColor=<rgba 0.6875 0 1 1>}
    Alpha=ScalarParameter() {group=Ice | 00_Globals, parameterName=Force Ice Mask to 1, sliderMax=1, sortPriority=10}
   )) {name="Ice Mask (S)", nodeColor=<rgba 0.2656 1 0 1>}
  n25 = NamedRerouteUsage()
  NamedRerouteDeclaration(None=Multiply(A=ScalarParameter() {group=Ice | 04_Volume, parameterName="Volume - Haze Amount (Alpha-Blending)", sortPriority=50}, B=OneMinus(None=@n25))) {name="Volume - Haze Amount (S)", nodeColor=<rgba 0 1 0.3438 1>}
  n26 = NamedRerouteUsage()
  n27 = Constant()
  n28 = NamedRerouteUsage()
  n29 = Constant() {r=1}
  MaterialExpressionReroute_3 = Reroute(None=If(A=NamedRerouteUsage(), B=NamedRerouteUsage(), "A > B"=AppendVector(A=@n26, B=@n27), "A < B"=AppendVector(A=@n28, B=@n29)))
  MaterialExpressionAppendVector_5 = AppendVector(A=Min(A=NamedRerouteUsage(), B=NamedRerouteUsage()), B=Saturate(None=NamedRerouteUsage()))
  n30 = Switch!(
   SwitchValue=Add(A=NamedRerouteUsage()) {constA=1}
   Default
   auto=@MaterialExpressionReroute_3
   outside=@MaterialExpressionAppendVector_5
   inside=@MaterialExpressionAppendVector_5
  ) {inputs [
    {
     inputName=auto
     input {
      expression=&self:MaterialExpressionReroute_3
      inputName=None
      mask=0
      maskA=0
      maskB=0
      maskG=0
      maskR=0
      outputIndex=0
     }
    }
    {
     inputName=outside
     input {
      expression=&self:MaterialExpressionAppendVector_5
      inputName=None
      mask=0
      maskA=0
      maskB=0
      maskG=0
      maskR=0
      outputIndex=0
     }
    }
    {
     inputName=inside
     input {
      expression=&self:MaterialExpressionAppendVector_5
      inputName=None
      mask=0
      maskA=0
      maskB=0
      maskG=0
      maskR=0
      outputIndex=0
     }
    }
   ]}
  MaterialExpressionStaticSwitch_0 = StaticSwitch(True=AppendVector(A=NamedRerouteUsage(), B=Constant() {r=1}), False=@n30, Value=NamedRerouteUsage())
  n31 = Convert(None=@MaterialExpressionStaticSwitch_0) {
   nodeName=Break Float2
   convertInputs [{
     type=Vector2
     expressionInput {
      expression=&self:MaterialExpressionStaticSwitch_0
      inputName=None
      mask=0
      maskA=0
      maskB=0
      maskG=0
      maskR=0
      outputIndex=0
     }
    }]
   convertOutputs [{type=Scalar}, {type=Scalar}]
  }
  NamedRerouteDeclaration(None=LinearInterpolate(A=@n31, Alpha=NamedRerouteUsage()) {constB=0}) {name="Is Scene Inside Volume (S)", nodeColor=<rgba 1 0 0.9844 1>}
  NamedRerouteDeclaration(None=@n31) {name="Volume Thickness (S)", nodeColor=<rgba 0 0.5312 1 1>}
  n32 = NamedRerouteUsage()
  n33 = NamedRerouteUsage()
  n34 = NamedRerouteUsage()
  MaterialExpressionStaticSwitchParameter_0 = StaticSwitchParameter(
   True=VectorParameter() {parameterName=MobileDistortion, defaultValue=<rgba 0.0879 0.1449 0.2812 1>}.RGBA
   False=ShadingPathSwitch(Default=StaticSwitch(True=@n32, False=@n33, Value=@n34), Mobile=NamedRerouteUsage())
  ) {parameterName=KFC_UseMobile}
  n35 = Convert(None=@MaterialExpressionStaticSwitchParameter_0) {
   convertInputs [{
     type=Vector4
     defaultValue=<rgba 0 0 0 1>
     expressionInput {
      expression=&self:MaterialExpressionStaticSwitchParameter_0
      inputName=None
      mask=0
      maskA=0
      maskB=0
      maskG=0
      maskR=0
      outputIndex=0
     }
    }]
   convertOutputs [{type=Vector3}, {type=Scalar}]
  }
  NamedRerouteDeclaration(None=@n35) {name="Refraction Result (V3)", nodeColor=<rgba 0 0.2435 1 1>}
  NamedRerouteDeclaration(None=@n35) {name="Material Opacity (S)", nodeColor=<rgba 0.9391 0 1 1>}
  n36 = ScalarParameter() {defaultValue=1, group=Ice | 02_Refraction, parameterName=Refraction - IOR, sliderMax=5, sliderMin=1, sortPriority=20}
  NamedRerouteDeclaration(None=@n36) {name="Refraction IOR (Constant) (S)", nodeColor=<rgba 0 0.1562 1 1>}
  n37 = SetMaterialAttributes!(
   MaterialAttributes
   BaseColor=Multiply(A=NamedRerouteUsage(), B=NamedRerouteUsage())
   Specular=NamedRerouteUsage()
   Roughness=NamedRerouteUsage()
   Normal=NamedRerouteUsage()
   EmissiveColor=Add(A=NamedRerouteUsage(), B=NamedRerouteUsage())
   Opacity=NamedRerouteUsage()
  ) {attributeSetTypes [
    #1
    #2
    #3
    #4
    #5
    #6
   ]}
  NamedRerouteDeclaration(None=@n37) {name="Dielectric Material (MA)", nodeColor=<rgba 1 0.7734 0 1>}
  NamedRerouteDeclaration(None=@n17.Specular Exp) {name="Specular Exp (S)", nodeColor=<rgba 1 0.2812 0 1>}
  NamedRerouteDeclaration(None=Constant() {r=10000}) {name="Scene Max Distance (S)", nodeColor=<rgba 0 0.0859 1 1>}
  NamedRerouteDeclaration(None=@n4.Fresnel Exp) {name="Ice Normal Fresnel Exp (S)", nodeColor=<rgba 1 0.0703 0 1>}
  NamedRerouteDeclaration(None=@n4."Max Intensity (Grazing)") {name="Ice Normal Max Intensity (S)", nodeColor=<rgba 0 0.25 1 1>}
  NamedRerouteDeclaration(None=@n4."Min Intensity (Facing)") {name="Ice Normal Min Intensity (S)", nodeColor=<rgba 0.5703 1 0 1>}
  n38 = GetMaterialAttributes(None=NamedRerouteDeclaration(None=FunctionInput() {inputName=Material, inputType=FunctionInput_MaterialAttributes}) {name="Material In (MA)", nodeColor=<rgba 0.2656 1 0 1>}) {attributeGetTypes [
    #1
    #2
    #3
    #5
    #4
   ]}
  NamedRerouteDeclaration(None=@n38.Specular) {name="Specular In (S)", nodeColor=<rgba 1 0 0.4922 1>}
  NamedRerouteDeclaration(None=ScalarParameter() {defaultValue=1, group=Ice | 02_Refraction, parameterName=Refraction - Distortion Scale, sortPriority=40}) {name="Refraction Distortion Scale (S)", nodeColor=<rgba 0.5 1 0 1>}
  NamedRerouteDeclaration(None=Max(A=ScalarParameter() {defaultValue=10, group=Ice | 03_Thickness, parameterName="Thickness - Max Thickness (Constant)", sortPriority=20}) {constB=0}) {name="Max Thickness (S)", nodeColor=<rgba 0.0547 0 1 1>}
  n39 = MaterialFunctionCall!(Normal=NamedRerouteUsage(), Direction=NamedRerouteUsage(), IOR=NamedRerouteUsage()) {materialFunction=/BRCosmetics/Characters/Player/Male/Medium/Bodies/M_MED_FoundKelp_Cast/Materials/ICE_DEV/MaterialFunctions/MF_ThicknessDirectional}
  n40 = Power(Base=@n39."Thickness Normalized (S)", Exp=NamedRerouteUsage())
  NamedRerouteDeclaration(None=Multiply(
    A=NamedRerouteDeclaration(None=@n40) {name="Asset Thickness VD [0, 1] (S)", desc=View Dependent, nodeColor=<rgba 0.4531 0 1 1>}
    B=NamedRerouteUsage()
   )) {name="Asset Thickness VD (S)", desc=View Dependent, nodeColor=<rgba 1 0 0.5156 1>}
  n41 = ScalarParameter() {
   group=Ice | 04_Volume
   parameterName="Refraction - Is Scene Inside Volume | -1: auto, 0: no, 1: yes"
   sliderMax=1
   sliderMin=-1
   sortPriority=60
  }
  NamedRerouteDeclaration(None=@n41) {name="Scene Inside Volume Method (S)", nodeColor=<rgba 1 0 0.1172 1>}
  NamedRerouteDeclaration(None=@n38.Normal) {name="Normal In TS (V3)", nodeColor=<rgba 1 0 0.4922 1>}
  NamedRerouteDeclaration(None=ScalarParameter() {defaultValue=1, group=Ice | 02_Refraction, parameterName=Refraction - Interface Thickness, sortPriority=10}) {name="Interface Thickness (S)", nodeColor=<rgba 1 0 0.7969 1>}
  NamedRerouteDeclaration(None=@n38.BaseColor) {name="BaseColor In (V3)", nodeColor=<rgba 1 0 0.4922 1>}
  n42 = StaticBoolParameter() {
   desc=Use for water planes or other assets representing interfaces to a volume with no backface or large enough that we never see the backface
   group=Ice | 03_Thickness
   parameterName=Infinite Thickness
   sortPriority=0
  }
  NamedRerouteDeclaration(None=@n42) {name="Infinite Thickness (SB)", nodeColor=<rgba 0 0.1562 1 1>}
  NamedRerouteDeclaration(None=@n38.Roughness) {name="Roughness In (S)", nodeColor=<rgba 1 0 0.4922 1>}
  n43 = MaterialFunctionCall!(Normal WS=NamedRerouteUsage(), Sky Illuminance Mult, Sky Desaturation, Sun Illuminance Mult, Sun Desaturation) {materialFunction=/Game/Global/MaterialFunctions/TechArt/Refraction/MF_SunSkyProperties}
  NamedRerouteDeclaration(None=@n43."Sun Direction (V3)") {name="Lighting - Sun Direction WS (V3)", nodeColor=<rgba 0 0.0625 1 1>}
  NamedRerouteDeclaration(None=@n43."Sun Illuminance (V3)") {name="Lighting - Sun Illuminance (V3)", nodeColor=<rgba 0.7344 0 1 1>}
  NamedRerouteDeclaration(None=@n43."Sky Illuminance (V3)") {name="Lighting - Sky Illuminance (V3)", nodeColor=<rgba 0 1 0.4141 1>}
  n44 = VectorParameter() {
   group=Ice | 02_Refraction
   parameterName="Refraction - Refracted Scene Avg Color [Opaque]"
   sortPriority=60
   defaultValue=<rgba 0.18 0.18 0.18 0>
  }
  NamedRerouteDeclaration(None=Saturate(None=@n44.RGB)) {name="Alpha Blending - Scene Avg Color (V3)", nodeColor=<rgba 0 0.625 1 1>}
  n45 = VectorParameter() {
   group=Ice | 02_Refraction
   parameterName="Refraction - Refracted Scene Scale [See-Through]"
   sortPriority=50
   channelNames {a=Intensity, b=Tint B, g=Tint G, r=Tint R}
   defaultValue=<rgba 1 1 1 1>
  }
  n46 = FunctionInput() {
   bUsePreviewValueAsDefault=true
   inputName=Refracted Scene Intensity Boost
   inputType=FunctionInput_Scalar
   sortPriority=50
   previewValue=<1 0 0 1>
  }
  n47 = Add(A=@n45.Intensity, B=@n46)
  NamedRerouteDeclaration(None=Max(A=Multiply(A=@n45.RGB, B=@n47)) {constB=0}) {name="Scene Tint (V3)", nodeColor=<rgba 0.7578 1 0 1>}
  n48 = LinearInterpolate(
   A=OneMinus(None=NamedRerouteUsage())
   Alpha=ScalarParameter() {group=Ice | 00_Globals, parameterName=Force Opaque Ice Mask to 1, sliderMax=1, sortPriority=20}
  )
  n49 = ScalarParameter() {group=Ice | 00_Globals, parameterName=Opaque Ice Mix, sliderMax=1, sortPriority=30}
  NamedRerouteDeclaration(None=StaticSwitch(True=@n48, False=Multiply(A=@n48, B=@n49), Value=NamedRerouteUsage())) {name="Opaque Mask (S)", nodeColor=<rgba 0.5937 1 0 1>}
  n50 = NamedRerouteUsage()
  n51 = ScalarParameter() {defaultValue=0.2, group=Ice | 05_Lighting, parameterName=Env Refraction Roughness, sliderMax=1, sortPriority=5}
  n52 = StaticBoolParameter() {group=Ice | 05_Lighting, parameterName=Env Illuminance Use Ground, sortPriority=40}
  n53 = ScalarParameter() {defaultValue=1, group=Ice | 05_Lighting, parameterName=Env Illuminance Ground Exp, sortPriority=50}
  n54 = ScalarParameter() {defaultValue=1, group=Ice | 05_Lighting, parameterName=Env Illuminance Ground Brightness, sortPriority=60}
  NamedRerouteDeclaration(None=Multiply(
    A=MaterialFunctionCall!(
     Direction WS=@n50
     Roughness=@n51
     Sky Desaturation
     Sky Brightness
     Use Ground Map=@n52
     Ground LatLong Map
     Group Map Rotation
     Ground Exp=@n53
     Ground Brightness=@n54
    ) {materialFunction=/Game/Global/MaterialFunctions/TechArt/Refraction/MF_EnvironmentIlluminance}."Environment Illuminance (V3)"
    B=ScalarParameter() {defaultValue=0.5, group=Ice | 05_Lighting, parameterName=Env Illuminance Mult, sortPriority=30}
   )) {name="Refracted Skybox (V3)", nodeColor=<rgba 0.125 1 0 1>}
  NamedRerouteDeclaration(None=@n38.EmissiveColor) {name="Emissive Color In (V3)", nodeColor=<rgba 1 0 0.4922 1>}
  n55 = TextureCoordinate()
  NamedRerouteDeclaration(None=Divide(A=NamedRerouteDeclaration(None=@n55) {name="Surface Base UV (V2)", nodeColor=<rgba 0 1 0.7422 1>}, B=NamedRerouteUsage())) {name="Surface Detail UV (V2)", nodeColor=<rgba 1 0.5859 0 1>}
  NamedRerouteDeclaration(None=StaticBoolParameter() {group=Ice | 02_Refraction, parameterName=Refraction - Enable Alpha Blending, sortPriority=0}) {name="Enable Alpha Blending (SB)", nodeColor=<rgba 0.0312 1 0 1>}
  NamedRerouteDeclaration(None=ScalarParameter() {defaultValue=1, group=Ice | 01_Surface, parameterName=Surface - Asset Normal Intensity, sortPriority=10}) {name="Asset Normal Intensity (S)", nodeColor=<rgba 0 1 0.9297 1>}
  n56 = VertexNormalWS()
  n57 = CameraVectorWS()
  n58 = OneMinus(None=Saturate(None=DotProduct(A=@n56, B=@n57)))
  n59 = NamedRerouteUsage()
  NamedRerouteDeclaration(None=LinearInterpolate(A=NamedRerouteUsage(), B=NamedRerouteUsage(), Alpha=Power(Base=@n58, Exp=@n59))) {name="Detail Normal Intensity (S)", nodeColor=<rgba 1 0.0234 0 1>}
  n60 = NamedRerouteUsage()
  n61 = NamedRerouteUsage()
  n62 = MaterialFunctionCall(
   BaseNormal=Normalize(VectorInput=MaterialFunctionCall!(Normal=@n60, Intensity=@n61) {materialFunction=/Game/Global/MaterialFunctions/MF_NormalIntensity}."Result (V3)")
   AdditionalNormal=NamedRerouteUsage()
  ) {desc=output is already normalized, materialFunction=&BlendAngleCorrectedNormals}
  n63 = MaterialFunctionCall!("Vector In (WS or TS)"=@n62.Result, Tangent WS, Bitangent WS, Normal WS) {materialFunction=/BRCosmetics/Characters/Player/Male/Medium/Bodies/M_MED_FoundKelp_Cast/Materials/ICE_DEV/MaterialFunctions/MF_TangentWorldTransforms}
  NamedRerouteDeclaration(None=@n63."Vector Out TS->WS (V3)") {name="Surface Normal WS (V3)", nodeColor=<rgba 0 0.5913 1 1>}
  n64 = VectorParameter() {
   group=Ice | 02_Refraction
   parameterName=Refraction - Far IOR
   sortPriority=30
   channelNames {a=Distance Range, b=Start Distance, g=Far IOR, r=Enable Far IOR}
   defaultValue=<rgba 0 1.01 100 400>
  }
  NamedRerouteDeclaration(None=Saturate(None=@n64.Enable Far IOR)) {name="Enable Far IOR (S)", nodeColor=<rgba 0 0.2969 1 1>}
  NamedRerouteDeclaration(None=Max(A=@n64.Start Distance) {constB=0}) {name="Far IOR Start Distance (S)", nodeColor=<rgba 0.6172 1 0 1>}
  NamedRerouteDeclaration(None=@n64.Distance Range) {name="Far IOR Distance Range (S)", nodeColor=<rgba 1 0 0.9375 1>}
  NamedRerouteDeclaration(None=@n64.Far IOR) {name="Far IOR (S)", nodeColor=<rgba 0 1 0.7422 1>}
  n65 = MaterialFunctionCall!(Viewport UV=NamedRerouteUsage()) {materialFunction=/Game/Global/MaterialFunctions/TechArt/Refraction/MF_DistanceToRefractedScene}
  NamedRerouteDeclaration(None=LinearInterpolate(A=NamedRerouteUsage(), B=@n65."Distance (S)", Alpha=SmoothStep(Value=@n65."Distance (S)") {constMax=0.01}) {desc="Avoids distance=0 when\r\nrefracting the skybox"}) {name="Scene Distance (Refraction) (S)", nodeColor=<rgba 0 0.487 1 1>}
  n66 = NamedRerouteUsage()
  n67 = NamedRerouteUsage()
  n68 = NamedRerouteUsage()
  n69 = NamedRerouteUsage()
  n70 = NamedRerouteUsage()
  n71 = Saturate(None=Divide(A=Subtract(A=@n68, B=@n69), B=Max(A=@n70) {constB=0.0001}))
  n72 = NamedRerouteUsage()
  NamedRerouteDeclaration(None=LinearInterpolate(A=@n66, B=Min(A=@n66, B=@n67), Alpha=Multiply(A=@n71, B=@n72))) {name="Refraction IOR (S)", nodeColor=<rgba 0 0.1562 1 1>}
  n73 = SceneColor!(UVs=NamedRerouteUsage())
  n74 = EyeAdaptation()
  NamedRerouteDeclaration(None=Multiply(A=NamedRerouteUsage(), B=Multiply(A=@n73.RGB, B=@n74.EyeAdaptation))) {name="Scene Luminance (Refraction) (V3)", nodeColor=<rgba 1 0.4453 0 1>}
  n75 = NamedRerouteUsage()
  NamedRerouteDeclaration(None=LinearInterpolate(A=NamedRerouteUsage(), B=@n75, Alpha=SmoothStep(Value=@n75) {constMax=0.01}) {desc="Avoids distance=0 when\r\nrefracting the skybox"}) {name="Scene Distance (Alpha Blending) (S)", nodeColor=<rgba 0 0.487 1 1>}
  NamedRerouteDeclaration(None=MaterialFunctionCall!(
    Sky Illuminance=NamedRerouteUsage()
    Sun Illuminance=NamedRerouteUsage()
    Sun Direction=NamedRerouteUsage()
    Scene Normal=Constant3Vector() {constant=<rgba 0 0 1 0>}
    Scene Albedo=NamedRerouteUsage()
    Scene Color Scale
   ) {materialFunction=/Game/Global/MaterialFunctions/TechArt/Refraction/MF_SceneLuminanceFromAlbedo}."Scene Luminance (V3)") {name="Scene Luminance (Alpha Blending) (V3)", nodeColor=<rgba 1 0.4453 0 1>}
  NamedRerouteDeclaration(None=MaterialFunctionCall() {materialFunction=&MF_DistanceToRefractedScene}."Distance (S)") {name="Distance to Scene (S)", nodeColor=<rgba 0.2187 1 0 1>}
  NamedRerouteDeclaration(None=ComponentMask(None=FunctionInput() {inputName=M, sortPriority=30}) {r=true}) {name="Surface AO In (S)", nodeColor=<rgba 0.2656 1 0 1>}
  n76 = Multiply(A=Subtract(A=NamedRerouteUsage(), B=NamedRerouteUsage()), B=NamedRerouteUsage())
  NamedRerouteDeclaration(None=Power(Base=Saturate(None=@n76), Exp=NamedRerouteUsage())) {name="Surface Opacity (S)", nodeColor=<rgba 0.0078 1 0 1>}
  n77 = NamedRerouteUsage()
  n78 = NamedRerouteUsage()
  NamedRerouteDeclaration(None=MaterialFunctionCall(Normal=TextureSample(UVs=@n77, Tex=@n78) {samplerSource=SSM_Wrap_WorldGroupSettings}.RGB, Intensity=NamedRerouteUsage()) {materialFunction=&MF_NormalIntensity}."Result (V3)") {name="Detail Normal TS (V3)", nodeColor=<rgba 1 0.0234 0 1>}
  n79 = VectorParameter() {
   group=Ice | 01_Surface
   parameterName=Surface - Opacity
   sortPriority=50
   channelNames {a=n/a, b=Exp, g=Multiply, r=Subtract}
   defaultValue=<rgba 0 1 1 0>
  }
  n80 = VectorParameter() {
   group=Ice | 01_Surface
   parameterName="Surface - Opacity (Mobile)"
   sortPriority=50
   channelNames {a=n/a, b=Exp, g=Multiply, r=Subtract}
   defaultValue=<rgba 0 1 1 0>
  }
  MaterialExpressionShadingPathSwitch_0 = ShadingPathSwitch(Default=@n79.RGBA, Mobile=@n80.RGBA)
  n81 = Convert(None=@MaterialExpressionShadingPathSwitch_0) {
   nodeName=Break Float4
   convertInputs [{
     type=Vector4
     defaultValue=<rgba 0 0 0 1>
     expressionInput {
      expression=&self:MaterialExpressionShadingPathSwitch_0
      inputName=None
      mask=0
      maskA=0
      maskB=0
      maskG=0
      maskR=0
      outputIndex=0
     }
    }]
   convertOutputs [{type=Scalar}, {type=Scalar}, {type=Scalar}, {type=Scalar}]
  }
  NamedRerouteDeclaration(None=@n81) {name="Surface Opacity Subtract (S)", nodeColor=<rgba 1 0 0.3281 1>}
  NamedRerouteDeclaration(None=@n62.Result) {name="Surface Normal TS (V3)", nodeColor=<rgba 0 0.8594 1 1>}
  NamedRerouteDeclaration(None=@n81) {name="Surface Opacity Multiply (S)", nodeColor=<rgba 0 0.6484 1 1>}
  n82 = Power(Base=NamedRerouteUsage(), Exp=NamedRerouteUsage())
  n83 = NamedRerouteUsage()
  NamedRerouteDeclaration(None=Saturate(None=Multiply(A=@n82, B=@n83))) {name="Surface Roughness (S)", nodeColor=<rgba 0 0.4844 1 1>}
  n84 = AppendVector(A=NamedRerouteUsage(), B=NamedRerouteUsage())
  n85 = LinearInterpolate(A=NamedRerouteUsage(), B=NamedRerouteUsage(), Alpha=NamedRerouteUsage())
  n86 = NamedRerouteUsage()
  MaterialExpressionShadingPathSwitch_1 = ShadingPathSwitch(Default=StaticSwitch(True=@n84, False=AppendVector(A=@n85, B=@n86), Value=NamedRerouteUsage()), Mobile=@n84)
  n87 = Convert(None=@MaterialExpressionShadingPathSwitch_1) {
   convertInputs [{
     type=Vector4
     defaultValue=<rgba 0 0 0 1>
     expressionInput {
      expression=&self:MaterialExpressionShadingPathSwitch_1
      inputName=None
      mask=0
      maskA=0
      maskB=0
      maskG=0
      maskR=0
      outputIndex=0
     }
    }]
   convertOutputs [{type=Vector3}, {type=Scalar}]
  }
  NamedRerouteDeclaration(None=LinearInterpolate(A=@n87, B=Max(A=@n87) {constB=100}, Alpha=NamedRerouteUsage())) {name="Scene Distance (S)", nodeColor=<rgba 0 0.487 1 1>}
  NamedRerouteDeclaration(None=@n87) {name="Scene Luminance (V3)", nodeColor=<rgba 1 0.4453 0 1>}
  NamedRerouteDeclaration(None=@n81) {name="Surface Opacity Exp (S)", nodeColor=<rgba 0.9688 1 0 1>}
  n88 = Multiply(A=Subtract(A=NamedRerouteUsage(), B=NamedRerouteUsage()), B=NamedRerouteUsage())
  NamedRerouteDeclaration(None=Power(Base=Saturate(None=@n88), Exp=NamedRerouteUsage())) {name="Surface AO (S)", nodeColor=<rgba 0.0078 1 0 1>}
  NamedRerouteDeclaration(None=@n17.Roughness Exp) {name="Roughness Exp (S)", nodeColor=<rgba 1 0.9844 0 1>}
  NamedRerouteDeclaration(None=@n17.Roughness Mult) {name="Roughness Mult (S)", nodeColor=<rgba 0.6641 0 1 1>}
 ]}
