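# Material function for the FoundKelp Cast player body cosmetic that adds a timed reactive glow, darkening BaseColor to black and blending Emissive toward a red color parameter.
# Time since a Blueprint-supplied timestamp, normalized by duration, samples a curve atlas, gated by an IsReactive flag with a manual debug override.
# This progress scales a TechArtMask channel blended toward half strength at Fresnel edges, forming the blend mask.
# It also outputs ReactivityProgress.
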MaterialFunction /BRCosmetics/Characters/Player/Male/Medium/Bodies/M_MED_FoundKelp_Cast/Materials/MF_FoundKelp_Cast_Reactivity
graph {nodes [
  n1 = GetMaterialAttributes(None=FunctionInput() {inputName=Material, inputType=FunctionInput_MaterialAttributes}) {attributeGetTypes [#1, #2]}
  n2 = NamedRerouteUsage()
  n3 = VectorParameter() {group=FoundKelp Cast, parameterName=FKC_ReactiveColor, defaultValue=<rgba 1 0.0057 0.0155 1>}
  FunctionOutput(None=SetMaterialAttributes!(
    MaterialAttributes=@n1.MaterialAttributes
    BaseColor=LinearInterpolate(A=@n1.BaseColor, Alpha=@n2) {constB=0}
    EmissiveColor=LinearInterpolate(A=@n1.EmissiveColor, B=@n3.RGB, Alpha=@n2) {constB=0}
   ) {attributeSetTypes [#1, #2]})
  FunctionOutput(None=NamedRerouteUsage()) {outputName=ReactivityProgress, sortPriority=1}
  n4 = Subtract(A=Time(), B=ScalarParameter() {group=FoundKelp Cast, parameterName=FKC_GameTimeFromBP})
  n5 = ScalarParameter() {defaultValue=5, group=FoundKelp Cast, parameterName=FKC_ReactivityDuration}
  n6 = CurveAtlasRowParameter(CurveTime=Saturate(None=Divide(A=@n4, B=@n5))) {
   atlas=/BRCosmetics/Characters/Player/Male/Medium/Bodies/M_MED_FoundKelp_Cast/Textures/CA_FoundKelp_Cast_Reactivity
   curve=/BRCosmetics/Characters/Player/Male/Medium/Bodies/M_MED_FoundKelp_Cast/Textures/C_FoundKelp_Cast_Reactivity
   group=FoundKelp Cast
   parameterName=FKC_ReactivityCurve
  }
  n7 = LinearInterpolate(
   A=Multiply(A=@n6.R, B=ScalarParameter() {group=FoundKelp Cast, parameterName=FKC_IsReactive})
   B=ScalarParameter() {defaultValue=1, group=FoundKelp Cast, parameterName=FKC_Reactivity}
   Alpha=ScalarParameter() {group=FoundKelp Cast, parameterName=FKC_ReactivityDebug}
  )
  n8 = TextureSample() {
   samplerType=SAMPLERTYPE_Masks
   texture=/BRCosmetics/Characters/Player/Male/Medium/Bodies/M_MED_FoundKelp_Cast/Textures/T_FoundKelp_Cast_Body_FX
  }
  n9 = ChannelMaskParameter(None=FunctionInput(Preview=@n8.RGB) {inputName=TechArtMask, sortPriority=1}) {group=FoundKelp Cast, maskChannel=Green, parameterName=FKC_ReactiveChannel, defaultValue=<rgba 0 1 0 0>}
  n10 = Fresnel() {baseReflectFraction=0, exponent=2}
  n11 = Multiply(A=@n9.R, B=OneMinus(None=@n10))
  n12 = Multiply(A=@n10) {constB=0.5}
  NamedRerouteDeclaration(None=Multiply(A=NamedRerouteDeclaration(None=@n7) {name=ReactiveProgress, nodeColor=<rgba 0.9219 0 1 1>}, B=Add(A=@n11, B=@n12))) {name=ReactiveMask, nodeColor=<rgba 0 0.7422 1 1>}
 ]}
