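# Translucent character material for the FoundKelp Cast outfit body, rendering volumetric ice inside the 2023 character parent PreFX/PostFX chain.
# MF_VolumetricIce takes tech-art transparency and squared-specular opacity inputs, then its opacity passes through translucency sorting.
# A chest-centered sphere mask on pre-skinned position, scaled by reactivity progress, forms a mask for boosting refracted scene intensity, while emissive blends toward boosted red.
# A mobile static switch adds a texture-driven opacity mask.
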Material /BRCosmetics/Characters/Player/Male/Medium/Bodies/M_MED_FoundKelp_Cast/Materials/ICE_DEV_UPDATE/M_FoundKelp_Cast_Ice
graph {
 properties {
  allowTranslucentCustomDepthWrites=true
  bAutomaticallySetUsageInEditor=false
  bScreenSpaceReflections=true
  bUseMaterialAttributes=true
  bUsedWithMeshDeformer=true
  bUsedWithMorphTargets=true
  bUsedWithSkeletalMesh=true
  blendMode=BLEND_Translucent
  translucencyLightingMode=TLM_SurfacePerPixelLighting
  usedShadingModels=Default Lit
 }
 nodes [
  n1 = MaterialFunctionCall() {materialFunction=&MF_CharacterParent_2023_PreFX}
  n2 = NamedRerouteUsage()
  n3 = ScalarParameter() {defaultValue=10, group=FoundKelp Cast, parameterName=FKC_ReactiveIceBoost}
  n4 = GetMaterialAttributes(None=MaterialFunctionCall!(
    Material=@n1.Result
    Ice Mask=NamedRerouteUsage()
    Surface Opacity=NamedRerouteUsage()
    M=@n1.Result
    Refracted Scene Intensity Boost=Multiply(A=@n2, B=@n3)
   ) {materialFunction=&/MF_VolumetricIce}.Material) {attributeGetTypes [#1, #2]}
  n5 = Constant3Vector() {constant=<rgba 6 0 0 0>}
  n6 = SetMaterialAttributes!(
   MaterialAttributes=@n4.MaterialAttributes
   Opacity=MaterialFunctionCall!(Opacity=@n4.Opacity) {materialFunction=/Game/Global/Materials/MaterialGym/Materials/SortedOutline/MF_SortTranslucency}.Result
   Emissive Color=LinearInterpolate(A=@n4.Emissive Color, B=Multiply(A=@n4.Emissive Color, B=@n5), Alpha=NamedRerouteUsage())
  ) {attributeSetTypes [#1, #2]}
  n7 = TextureSample() {
   samplerType=SAMPLERTYPE_Masks
   texture=/BRCosmetics/Characters/Player/Male/Medium/Bodies/M_MED_FoundKelp_Cast/Textures/T_FoundKelp_Cast_Body_TA
  }
  n8 = OneMinus(None=@n7.G)
  MaterialFunctionCall(
   Material=StaticSwitchParameter(True=SetMaterialAttributes!(MaterialAttributes=@n6, Opacity Mask=@n8) {attributeSetTypes [#3]}, False=@n6) {parameterName=KFC_UseMobile}
   M=@n1.M
   Pre Skinned Local Position=@n1.M
   Distance Blend=@n1.M
   Fresnel=@n1.M
  ) {materialFunction=&MF_CharacterParent_2023_PostFX}
  n9 = SphereMask(
   A=NamedRerouteUsage()
   B=VectorParameter() {group=FoundKelp Cast, parameterName=ChestLoc, defaultValue=<rgba 8 6 130.5 1>}.RGB
   Radius=ScalarParameter() {defaultValue=11, group=FoundKelp Cast, parameterName=Radius}
  ) {hardnessPercent=0}
  n10 = Multiply(A=@n9) {constB=2}
  MaterialFunctionCall(Alpha Threshold=OneMinus(None=Saturate(None=@n10))) {materialFunction=&DitherTemporalAA}
  NamedRerouteDeclaration(None=ChannelMaskParameter(None=@n1.Tech Art Mask) {group=FoundKelp Cast, parameterName=TransparentIceMask}.R) {name=TransparentIceMask, nodeColor=<rgba 0.8984 0 1 1>}
  n11 = TextureSampleParameter2D() {
   desc="R = Spec , G = Roughness, B= Metal"
   group=00_Texture Inputs
   parameterName=SRM
   samplerSource=SSM_Wrap_WorldGroupSettings
   samplerType=SAMPLERTYPE_Masks
   sortPriority=1
   texture=&T_Fortnite_Default_S
  }
  NamedRerouteDeclaration(None=Multiply(A=@n11.R, B=@n11.R)) {name="Ice Surface Opacity (S)", nodeColor=<rgba 0 1 0.4141 1>}
  NamedRerouteDeclaration(None=@n1.Pre skinned Local Position) {name=PreskinnedLocPos, nodeColor=<rgba 1 0.6797 0 1>}
  NamedRerouteDeclaration(None=Multiply(
    A=MaterialFunctionCall!(Material=@n1.Result, TechArtMask=@n1.Result) {materialFunction=/BRCosmetics/Characters/Player/Male/Medium/Bodies/M_MED_FoundKelp_Cast/Materials/MF_FoundKelp_Cast_Reactivity}.ReactivityProgress
    B=@n9
   )) {name=RefractedSceneIntensityMask, nodeColor=<rgba 0 1 0.8828 1>}
  TemporalResponsivenessOutput!(None=Constant())
 ]
}
